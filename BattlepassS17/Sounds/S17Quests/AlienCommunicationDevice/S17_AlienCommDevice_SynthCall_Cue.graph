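# S17_AlienCommDevice_SynthCall_Cue is a SoundCue for the synth call of the alien communication device in the Season 17 Battle Pass quests.
# A Delay node at the root holds playback for a fixed interval before a single Wave Player plays the SynthCall sound wave.
# The cue is routed to the Fort_Quests sound class, plays at reduced volume, and uses a shared comm-quest attenuation asset for distance falloff.
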SoundCue /BattlepassS17/Sounds/S17Quests/AlienCommunicationDevice/S17_AlienCommDevice_SynthCall_Cue
root=Delay_0
nodes [
 {id=Delay_0, children [WavePlayer_0], properties {delayMax=1, delayMin=1}}
 {id=WavePlayer_0, properties {soundWaveAssetPtr=&/S17_AlienCommDevice_SynthCall}}
]
properties {
 attenuationSettings=/BattlepassS17/Sounds/S17_Comm_quest_Att.S17_Comm_Quest_Att
 maxDistance=4000
 soundClassObject=&Fort_Quests
 volumeMultiplier=0.5
}
